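# Sound cue for the swing of the MechPilot cosmetic pickaxe, blending close and medium-distance layers through a mixer.
# The close attenuation branch plays MechPilot-specific swing waves, while the medium branch reuses generic Fire Axe distant swings.
# Each branch runs through a modulator and a quality switch that steps from six random variants down to three, then a single wave.
# The cue primes on load and overrides concurrency with a small per-owner limit that stops the oldest instance.
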SoundCue /BRCosmetics/Sounds/PickAxes/MechPilot/PickAxe_MechPilot_Swing_Athena_Cue
root=Mixer_1
nodes [
 {id=Attenuation_2, children [Modulator_3], properties {attenuationSettings=/Game/Sounds/Attenuation/Fort_Weapon_Pickaxe_Close_Att}}
 {id=Attenuation_3, children [Modulator_2], properties {attenuationSettings=/Game/Sounds/Attenuation/Fort_Weapon_Pickaxe_Medium_Att}}
 {id=Mixer_1, children [Attenuation_2, Attenuation_3], properties {inputVolume [1,1]}}
 {id=Modulator_2, children [QualityLevel_2]}
 {id=Modulator_3, children [QualityLevel_3]}
 {id=QualityLevel_2, children [Random_2, Random_4, WavePlayer_28, WavePlayer_28]}
 {id=QualityLevel_3, children [Random_5, Random_6, WavePlayer_38, WavePlayer_38]}
 {
  id=Random_2
  children [WavePlayer_0, WavePlayer_3, WavePlayer_5, WavePlayer_14, WavePlayer_15, WavePlayer_17]
  properties {weights [1,1,1,1,1,1]}
 }
 {id=Random_4, children [WavePlayer_23, WavePlayer_22, WavePlayer_24], properties {weights [1,1,1]}}
 {
  id=Random_5
  children [WavePlayer_29, WavePlayer_30, WavePlayer_31, WavePlayer_32, WavePlayer_33, WavePlayer_34]
  properties {weights [1,1,1,1,1,1]}
 }
 {id=Random_6, children [WavePlayer_35, WavePlayer_36, WavePlayer_37], properties {weights [1,1,1]}}
 {id=WavePlayer_0, properties {soundWaveAssetPtr=/Game/Athena/Sounds/Weapons/Fire_Axe_Swing_Distant_04}}
 {id=WavePlayer_14, properties {soundWaveAssetPtr=/Game/Athena/Sounds/Weapons/Fire_Axe_Swing_Distant_03}}
 {id=WavePlayer_15, properties {soundWaveAssetPtr=/Game/Athena/Sounds/Weapons/Fire_Axe_Swing_Distant_05}}
 {id=WavePlayer_17, properties {soundWaveAssetPtr=/Game/Athena/Sounds/Weapons/Fire_Axe_Swing_Distant_06}}
 {id=WavePlayer_22, properties {soundWaveAssetPtr=&Fire_Axe_Swing_Distant_03}}
 {id=WavePlayer_23, properties {soundWaveAssetPtr=&Fire_Axe_Swing_Distant_04}}
 {id=WavePlayer_24, properties {soundWaveAssetPtr=&Fire_Axe_Swing_Distant_06}}
 {id=WavePlayer_28, properties {soundWaveAssetPtr=&Fire_Axe_Swing_Distant_06}}
 {id=WavePlayer_29, properties {soundWaveAssetPtr=&/PickAxe_MechPilot_Swing_01}}
 {id=WavePlayer_3, properties {soundWaveAssetPtr=/Game/Athena/Sounds/Weapons/Fire_Axe_Swing_Distant_01}}
 {id=WavePlayer_30, properties {soundWaveAssetPtr=&/PickAxe_MechPilot_Swing_02}}
 {id=WavePlayer_31, properties {soundWaveAssetPtr=&/PickAxe_MechPilot_Swing_03}}
 {id=WavePlayer_32, properties {soundWaveAssetPtr=&/PickAxe_MechPilot_Swing_04}}
 {id=WavePlayer_33, properties {soundWaveAssetPtr=&/PickAxe_MechPilot_Swing_05}}
 {id=WavePlayer_34, properties {soundWaveAssetPtr=&/PickAxe_MechPilot_Swing_06}}
 {id=WavePlayer_35, properties {soundWaveAssetPtr=&PickAxe_MechPilot_Swing_01}}
 {id=WavePlayer_36, properties {soundWaveAssetPtr=&PickAxe_MechPilot_Swing_02}}
 {id=WavePlayer_37, properties {soundWaveAssetPtr=&PickAxe_MechPilot_Swing_03}}
 {id=WavePlayer_38, properties {soundWaveAssetPtr=&PickAxe_MechPilot_Swing_01}}
 {id=WavePlayer_5, properties {soundWaveAssetPtr=/Game/Athena/Sounds/Weapons/Fire_Axe_Swing_Distant_02}}
]
properties {
 audioPropertiesSheet=/BRCosmetics/Sounds/Settings/PropertySheets/Pickaxe/APS_Cosmetic_Pickaxe_Swing
 bOverrideConcurrency=true
 bPrimeOnLoad=true
 maxDistance=3000
 soundClassObject=&Fort_WeaponMelee_PSM
 volumeMultiplier=0.2
 concurrencyOverrides {bLimitToOwner=true, maxCount=2, resolutionRule=StopOldest}
 concurrencySet [&Player_Melee_Weapon_Conc]
}
